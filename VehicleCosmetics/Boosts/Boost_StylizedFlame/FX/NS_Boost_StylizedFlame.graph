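# Niagara system for the Stylized Flame vehicle boost cosmetic, emitting flames from vehicle skeletal-mesh sockets.
# A sprite emitter spawns SubUV flame puffs tuned per spawn group, while a ribbon emitter draws solid flame trails per port with animated secondary widths.
# Painted-variant overrides replace default color gradients when painted, and a frontend-preview flag adjusts offsets, lifetimes and velocities.
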NiagaraSystem /VehicleCosmetics/Boosts/Boost_StylizedFlame/FX/NS_Boost_StylizedFlame
dependencies {
 usedDataInterfaces [
  &NiagaraDataInterfaceCurve
  &NiagaraDataInterfaceSpriteRendererInfo
  &NiagaraDataInterfaceColorCurve
  &NiagaraDataInterfaceArrayFloat
  &NiagaraDataInterfaceSkeletalMesh
 ]
 usedDynamicInputs []
 usedModules [
  &self:SystemSpawnScript.SystemScriptSource.SystemScriptGraph.NiagaraNodeAssignment_0.SetVariables_#1
  &SystemState
 ]
}
emitters {
 NE_VC_Boost_BaseFlames {
  emitterUpdateScript [
   &EmitterState(Life Cycle Mode=::NewEnumerator1, UseLoopDelay=false)
   &VC_FilteredSocketCount()
   &SpawnRate(Spawn Probability=0.6, SpawnRate=20, Use Spawn Probability=true)
   &SpawnRate(SpawnGroup=1, SpawnRate=18, Use Spawn Probability=false) {module=SpawnRate001}
  ]
  particleSpawnScript [
   &VC_PaintedVariantsOverride(
    Black=<rgba 0.19 0.19 0.19 8>
    BurntSienna=<rgba 0.65 0.2907 0.0779 8>
    Cobalt=<rgba 0.2844 0.4788 1.4 8>
    Crimson=<rgba 1.3 0.2505 0.2505 8>
    DefaultColor=<rgba 1 1 1 5>
    ForestGreen=<rgba 0.5604 1.1 0.3781 8>
    Grey=<rgba 0.4062 0.4062 0.4062 8>
    Lime=<rgba 0.8967 1 0.3802 8>
    Orange=<rgba 1.4 0.3872 0.1823 8>
    Pink=<rgba 1.4738 0.6094 1.5 8>
    Purple=<rgba 0.5558 0.1302 1 8>
    R1 Black:LinearColor=<rgba 0.05 0.05 0.05 2.6>
    R1 BurntSienna:LinearColor=<rgba 0.18 0.0764 0.0302 2.5>
    R1 Cobalt:LinearColor=<rgba 0.0673 0.1695 0.6458 3>
    R1 Crimson:LinearColor=<rgba 0.4219 0.0835 0.0835 2.5>
    R1 DefaultColor=<rgba 0 0 0 1>
    R1 ForestGreen:LinearColor=<rgba 0.1945 0.4427 0.1107 2.5>
    R1 Grey:LinearColor=<rgba 0.1094 0.1094 0.1094 2.5>
    R1 Lime:LinearColor=<rgba 0.3389 0.3854 0.1064 2.5>
    R1 Orange:LinearColor=<rgba 0.5208 0.1287 0.0651 2.5>
    R1 Pink:LinearColor=<rgba 0.526 0.1644 0.4959 2>
    R1 Purple:LinearColor=<rgba 0.1676 0.0072 0.3177 2>
    R1 Saffron:LinearColor=<rgba 0.5 0.4872 0.1432 2>
    R1 SkyBlue:LinearColor=<rgba 0.0752 0.514 0.6875 2>
    R1 TitaniumWhite:LinearColor=<rgba 0.474 0.474 0.474 3.5>
    Saffron=<rgba 1.15 1.117 0.2336 8>
    SkyBlue=<rgba 0.4167 1.2647 1.6 8>
    TitaniumWhite=<rgba 1.1 1.1 1.1 8>
    Use R1 Painted Overrides=true
   )
   /SetVariables(Particles.RENDERTAG:NiagaraInt32=&SpawnGroup()) {isSetParameters=true}
   &InitializeParticle(
    Color Mode=::NewEnumerator0
    Lifetime=/VehicleCosmetics/FX/Niagara/Modules/MergeEmitters/VC_FloatBySpawnGroup(
     SpawnGroup or ExecIndex:NiagaraInt32=0
     CustomInt:NiagaraInt32=0
     MaximumSpawnGroups:NiagaraInt32=3
     Float SG 0:NiagaraFloat=&MakeCustomFloatFromBool(
      Bool=$User.bIsFrontendPreview
      True Float=0.65
      False Float=&RandomRangeFloat(Minimum=0.4, Maximum=0.45)
     )
     Float SG 1:NiagaraFloat=&MakeCustomFloatFromBool(
      Bool=$User.bIsFrontendPreview
      True Float=0.65
      False Float=&RandomRangeFloat(Minimum=0.4, Maximum=0.45)
     )
     Float SG 2:NiagaraFloat=&RandomRangeFloat(Minimum=0.08, Maximum=0.11)
     Float SG 3:NiagaraFloat=0
     Float SG 4:NiagaraFloat=0
     Float SG 5:NiagaraFloat=0
    )
    Sprite Rotation Mode=::NewEnumerator2
    Sprite Size Mode=::NewEnumerator3
    Uniform Sprite Size=&VC_FloatBySpawnGroup(
     SpawnGroup or ExecIndex:NiagaraInt32=0
     CustomInt:NiagaraInt32=0
     MaximumSpawnGroups:NiagaraInt32=3
     Float SG 0:NiagaraFloat=61
     Float SG 1:NiagaraFloat=61
     Float SG 2:NiagaraFloat=&RandomRangeFloat(Minimum=8, Maximum=13)
     Float SG 3:NiagaraFloat=0
     Float SG 4:NiagaraFloat=0
     Float SG 5:NiagaraFloat=0
    )
    UsePositionOffset=false
    Write Material Random=true
   )
   &PartitionParticles(Number Of Partitions=$Emitter.PortCount, Particles Per Partition=2, Partition Mode:ENiagara_PartitionMode=::NewEnumerator1)
   &SkeletalMeshLocation(
    Bone / Socket Index=$Particles.PartitionParticles.Partition
    Bone / Socket Sampling Mode=::NewEnumerator5
    MaskSpawnGroup=false
    Mesh Sampling Type=::NewEnumerator6
    Sampled Position Offset=/VehicleCosmetics/FX/Niagara/Modules/MergeEmitters/VC_VectorBySpawnGroup(
     SpawnGroup or ExecIndex:NiagaraInt32=0
     CustomInt:NiagaraInt32=0
     MaximumSpawnGroups:NiagaraInt32=3
     Vector Group 0:Vector3f=&AddVectorToPosition(
      Position=/Niagara/DynamicInputs/TypeConversions/MakeCustomPositionFromBool(Bool:NiagaraBool=$User.bIsFrontendPreview, True Position:NiagaraPosition=<-5 0 1>, False Position:NiagaraPosition=<-10 0 0>)
      Vector=$User.DTCarOffset
     )
     Vector Group 1:Vector3f=&AddVectorToPosition(
      Position=&MakeCustomPositionFromBool(Bool:NiagaraBool=$User.bIsFrontendPreview, True Position:NiagaraPosition=<-5 0 1>, False Position:NiagaraPosition=<-10 0 0>)
      Vector=$User.DTCarOffset
     )
     Vector Group 2:Vector3f=&AddVectorToPosition(Position=<-47 0 0>, Vector=$User.DTCarOffset)
     Vector Group 3:Vector3f=<0 0 0>
     Vector Group 4:Vector3f=<0 0 0>
     Vector Group 5:Vector3f=<0 0 0>
    )
    Skeletal Mesh=$System.SkeletalMesh
   )
   &AddVelocity(
    Velocity=&VC_VectorBySpawnGroup(
     SpawnGroup or ExecIndex:NiagaraInt32=0
     CustomInt:NiagaraInt32=0
     MaximumSpawnGroups:NiagaraInt32=3
     Vector Group 0:Vector3f=&MakeCustomVectorFromBool(Bool=$User.bIsFrontendPreview, True Vector=<-170 0 0>, False Vector=<-50 0 0>)
     Vector Group 1:Vector3f=&MakeCustomVectorFromBool(Bool=$User.bIsFrontendPreview, True Vector=<-170 0 0>, False Vector=<-50 0 0>)
     Vector Group 2:Vector3f=<-1000 0 0>
     Vector Group 3:Vector3f=<0 0 0>
     Vector Group 4:Vector3f=<0 0 0>
     Vector Group 5:Vector3f=<0 0 0>
    )
    Velocity Speed Scale=&VC_FloatBySpawnGroup(
     SpawnGroup or ExecIndex:NiagaraInt32=0
     CustomInt:NiagaraInt32=0
     MaximumSpawnGroups:NiagaraInt32=3
     Float SG 0:NiagaraFloat=1
     Float SG 1:NiagaraFloat=1
     Float SG 2:NiagaraFloat=1
     Float SG 3:NiagaraFloat=0
     Float SG 4:NiagaraFloat=0
     Float SG 5:NiagaraFloat=0
    )
   )
   &ShapeLocation(
    MaskSpawnGroup=false
    Override Seed=false
    Sphere Radius=&VC_FloatBySpawnGroup(
     SpawnGroup or ExecIndex:NiagaraInt32=0
     CustomInt:NiagaraInt32=0
     MaximumSpawnGroups:NiagaraInt32=3
     Float SG 0:NiagaraFloat=0
     Float SG 1:NiagaraFloat=0
     Float SG 2:NiagaraFloat=0
     Float SG 3:NiagaraFloat=0
     Float SG 4:NiagaraFloat=0
     Float SG 5:NiagaraFloat=0
    )
    UseOwnerScale=true
   )
   &CameraOffset(Camera Offset Amount=&VC_FloatBySpawnGroup(
     SpawnGroup or ExecIndex:NiagaraInt32=0
     CustomInt:NiagaraInt32=0
     MaximumSpawnGroups:NiagaraInt32=3
     Float SG 0:NiagaraFloat=20
     Float SG 1:NiagaraFloat=20
     Float SG 2:NiagaraFloat=0
     Float SG 3:NiagaraFloat=0
     Float SG 4:NiagaraFloat=0
     Float SG 5:NiagaraFloat=0
    ))
  ]
  particleUpdateScript [
   &ParticleState()
   &ScaleSpriteSize(
    Scale Sprite Size Mode=::NewEnumerator0
    Uniform Scale Factor=&VC_FloatBySpawnGroup(
     SpawnGroup or ExecIndex:NiagaraInt32=0
     CustomInt:NiagaraInt32=0
     MaximumSpawnGroups:NiagaraInt32=3
     Float SG 0:NiagaraFloat=&FloatFromCurve(FloatCurve=DataInterface() {propertyValues {Curve {keys [{tangentMode=RCTM_User, time=0, value=1}, {tangentMode=RCTM_User, time=1, value=2.5}]}}})
     Float SG 1:NiagaraFloat=&FloatFromCurve(FloatCurve=DataInterface() {propertyValues {Curve {keys [{tangentMode=RCTM_User, time=0, value=1}, {tangentMode=RCTM_User, time=1, value=2.5}]}}})
     Float SG 2:NiagaraFloat=1
     Float SG 3:NiagaraFloat=0
     Float SG 4:NiagaraFloat=0
     Float SG 5:NiagaraFloat=0
    )
   ) {module=ScaleSpriteSize001}
   &SubUVAnimation(
    End Frame Range Override=26
    "Sprite Renderer "=DataInterface() {propertyValues {SpriteRenderer=&self:NE_VC_Boost_BaseFlames_6.NiagaraSpriteRendererProperties_4}}
    UseEndFrame=false
    UseStartFrame=false
   )
   &SubUVAnimation(
    End Frame Range Override=26
    "Sprite Renderer "=DataInterface() {propertyValues {SpriteRenderer=&self:NE_VC_Boost_BaseFlames_6.NiagaraSpriteRendererProperties_3}}
    SubUV Animation Mode:ENiagara_SubUVLookupModeV2=::NewEnumerator1
    SubUV Curve=DataInterface() {propertyValues {Curve {
       defaultValue=0
       keys [
        {arriveTangent=0.5133, interpMode=RCIM_Cubic, leaveTangent=0.5133, tangentMode=RCTM_User, time=0, value=0}
        {interpMode=RCIM_Cubic, tangentMode=RCTM_User, time=0.85, value=1}
       ]
      }}}
    UseEndFrame=false
    UseStartFrame=false
   ) {module=SubUVAnimation001}
   &DynamicMaterialParameters(
    Index 0 Param 1=0.01
    Index 0 Param 2=&RandomRangeFloat(Minimum=0, Maximum=0.075)
    Index 0 Param 3=&RandomRangeFloat(Minimum=-0.2, Maximum=0.78)
    Index 0 Param 4=0.025
    Module.Index0Param0Write=true
    Module.Index0Param1Write=true
    Module.Index0Param2Write=true
    Module.Index0Param3Write=true
   )
   &Drag(
    Drag=&VC_FloatBySpawnGroup(
     SpawnGroup or ExecIndex:NiagaraInt32=0
     CustomInt:NiagaraInt32=0
     MaximumSpawnGroups:NiagaraInt32=3
     Float SG 0:NiagaraFloat=0
     Float SG 1:NiagaraFloat=0
     Float SG 2:NiagaraFloat=0
     Float SG 3:NiagaraFloat=0
     Float SG 4:NiagaraFloat=0
     Float SG 5:NiagaraFloat=0
    )
    Rotational Drag:NiagaraFloat=&VC_FloatBySpawnGroup(
     SpawnGroup or ExecIndex:NiagaraInt32=0
     CustomInt:NiagaraInt32=0
     MaximumSpawnGroups:NiagaraInt32=3
     Float SG 0:NiagaraFloat=0
     Float SG 1:NiagaraFloat=0
     Float SG 2:NiagaraFloat=0
     Float SG 3:NiagaraFloat=0
     Float SG 4:NiagaraFloat=0
     Float SG 5:NiagaraFloat=0
    )
    Use Linear Drag=true
    Use Rotational Drag=true
   )
   &InheritSourceMovement(
    Applied Position Scale=&VC_VectorBySpawnGroup(
     SpawnGroup or ExecIndex:NiagaraInt32=0
     CustomInt:NiagaraInt32=0
     MaximumSpawnGroups:NiagaraInt32=3
     Vector Group 0:Vector3f=<0 0 0>
     Vector Group 1:Vector3f=<0 0 0>
     Vector Group 2:Vector3f=&Multiply_VectorByFloat(
      Vector=<1 1 1>
      Float=&FloatFromCurve(FloatCurve=DataInterface() {propertyValues {Curve {keys [
           {
            arriveTangentWeight=0.4252
            interpMode=RCIM_Cubic
            leaveTangentWeight=0.4252
            tangentMode=RCTM_User
            tangentWeightMode=RCTWM_WeightedBoth
            time=0.05
            value=1
           }
           {interpMode=RCIM_Cubic, tangentMode=RCTM_User, time=1, value=0}
          ]}}})
     )
     Vector Group 3:Vector3f=<0 0 0>
     Vector Group 4:Vector3f=<0 0 0>
     Vector Group 5:Vector3f=<0 0 0>
    )
    ApplyForce=false
    ApplyPosition=true
    ApplyPositionFalloff=false
   )
   &SolveForcesAndVelocity(Clamp Velocity=false, Limit Acceleration=false)
   &ScaleColor(
    Scale Alpha=&FloatFromCurve(FloatCurve=DataInterface() {propertyValues {Curve {keys [<at 0 1>,<at 0.85 0>]}}})
    Scale RGB=&MakeCustomVectorFromBool(
     Bool=$User.bIsPainted
     True Vector=&Multiply_VectorByFloat(
      Vector=<1 1 1>
      Float=&FloatFromCurve(
       FloatCurve=DataInterface() {propertyValues {Curve {keys [
           {arriveTangent=-0.298, interpMode=RCIM_Cubic, leaveTangent=-0.298, tangentMode=RCTM_User, time=0, value=2.3}
           {arriveTangent=-1.3268, interpMode=RCIM_Cubic, leaveTangent=-2.1054, tangentMode=RCTM_Break, time=0.3, value=1.75}
           {arriveTangent=-1.9879, interpMode=RCIM_Cubic, leaveTangent=-1.9879, tangentMode=RCTM_User, time=0.55, value=1.25}
          ]}}}
       Scale Curve=0.3
      )
     )
     False Vector=&MakeCustomVectorFromBool(
      Bool=$User.bIsFrontendPreview
      True Vector=&MakeVectorFromLinearColorRGB(Color=&ColorFromCurve(ColorCurve=DataInterface() {propertyValues {
          AlphaCurve {keys [<at 0 1 RCIM_Cubic>,<at 0.6923 1>,<at 1 1>]}
          BlueCurve {keys [<at 0.3366 0.4948 RCIM_Cubic>,<at 0.3766 0.2452>,<at 0.5906 0>]}
          GreenCurve {keys [<at 0.3366 0.3273 RCIM_Cubic>,<at 0.3766 0.2585>,<at 0.5906 0.0667>]}
          RedCurve {keys [<at 0.3366 0.1484 RCIM_Cubic>,<at 0.3766 0.4635>,<at 0.5906 1>]}
         }}))
      False Vector=&MakeVectorFromLinearColorRGB(Color=&ColorFromCurve(ColorCurve=DataInterface() {propertyValues {
          AlphaCurve {keys [<at 0 1 RCIM_Cubic>,<at 0.9703 1>]}
          BlueCurve {keys [<at 0.3366 0.4948 RCIM_Cubic>,<at 0.3766 0.2452>,<at 0.5906 0>]}
          GreenCurve {keys [<at 0.3366 0.3273 RCIM_Cubic>,<at 0.3766 0.2585>,<at 0.5906 0.0667>]}
          RedCurve {keys [<at 0.3366 0.1484 RCIM_Cubic>,<at 0.3766 0.4635>,<at 0.5906 1>]}
         }}))
     )
    )
    ScaleA=true
    ScaleRGB=true
   )
   /SetVariables(
    Particles.R1 Color:LinearColor=&Multiply_LinearColorByFloat(
     Linear Color=&MakeCustomLinearColorFromBool(
      Bool=$User.bIsPainted
      True Color=$Particles.Initial.R1 Color
      False Color=&ColorFromCurve(ColorCurve=DataInterface() {propertyValues {
         AlphaCurve {keys [<at 0 5 RCIM_Cubic>,<at 1 3 RCIM_Cubic>]}
         BlueCurve {keys [<at 0.2395 0.5 RCIM_Cubic>,<at 0.3766 0.2455>,<at 0.5906 0>]}
         GreenCurve {keys [<at 0.2395 0.3629 RCIM_Cubic>,<at 0.3766 0.2588>,<at 0.5906 0.1414>]}
         RedCurve {keys [<at 0.2395 0.2213 RCIM_Cubic>,<at 0.3766 0.464>,<at 0.5906 0.8>]}
        }})
     )
     Float=0.9
    )
    Particles.R2 Color:LinearColor=&MakeCustomLinearColorFromBool(
     Bool=$User.bIsPainted
     True Color=&VC_DynamicScaleColor(
      Color to Scale:LinearColor=$Particles.Initial.R1 Color
      RGB Scale:Vector3f=&Multiply_VectorByFloat(Vector=<1 1 1>, Float=0.25)
      Alpha Scale:NiagaraFloat=1
     )
     False Color=&VC_DynamicScaleColor(Color to Scale:LinearColor=$Particles.Initial.R2 Color, RGB Scale:Vector3f=<1 1 1>, Alpha Scale:NiagaraFloat=1)
    )
   ) {isSetParameters=true}
  ]
  rendererClasses [&NiagaraSpriteRendererProperties]
  renderers [
   {rendererClass=&NiagaraSpriteRendererProperties}
   {rendererClass=&NiagaraSpriteRendererProperties, rendererIndex=1}
   {rendererClass=&NiagaraSpriteRendererProperties, rendererIndex=2}
  ]
 }
 NE_VC_SolidFlames_01 {
  emitterUpdateScript [
   &EmitterState(Life Cycle Mode=::NewEnumerator1, Loop Duration=1, UseLoopDelay=false)
   &VC_FilteredSocketCount()
   &VC_RateQualityHandler(Default Rate=100)
   &VC_SpawnRibbon_v1(Number of Ribbons=$Emitter.PortCount, SpawnRate=$Emitter.Spawn Rate)
  ]
  particleSpawnScript [
   &VC_PaintedVariantsOverride(
    Black=<rgba 0.0075 0.0075 0.0075 1>
    BurntSienna=<rgba 0.33 0.0828 0.0165 4>
    Cobalt=<rgba 0.045 0.1541 1.5 2>
    Crimson=<rgba 0.65 0.052 0.052 2>
    DefaultColor=<rgba 0.4 0.0263 0.02 8>
    ForestGreen=<rgba 0.04 0.4 0.052 2>
    Grey=<rgba 0.058 0.058 0.058 2>
    Lime=<rgba 0.75 1.5 0 4>
    Orange=<rgba 0.8 0.08 0 1.5>
    Pink=<rgba 0.875 0.25 1 2.5>
    Purple=<rgba 0.2167 0 0.65 2.3>
    R1 Black:LinearColor=<rgba 0.035 0.035 0.035 1>
    R1 BurntSienna:LinearColor=<rgba 0.5 0.0998 0.01 5>
    R1 Cobalt:LinearColor=<rgba 0.045 0.192 2.25 1>
    R1 Crimson:LinearColor=<rgba 1.2 0.084 0.1026 1>
    R1 DefaultColor=<rgba 2 0.5 0 8>
    R1 ForestGreen:LinearColor=<rgba 0.3362 1 0.115 1>
    R1 Grey:LinearColor=<rgba 0.13 0.13 0.13 2>
    R1 Lime:LinearColor=<rgba 0.7875 1 0.15 1.5>
    R1 Orange:LinearColor=<rgba 1.25 0.2087 0.025 1>
    R1 Pink:LinearColor=<rgba 1.1333 0.25 1.25 2>
    R1 Purple:LinearColor=<rgba 0.3724 0.038 0.95 1.5>
    R1 Saffron:LinearColor=<rgba 1 1 0.2 1.5>
    R1 SkyBlue:LinearColor=<rgba 0.18 0.99 1.8 1.5>
    R1 TitaniumWhite:LinearColor=<rgba 0.75 0.75 0.75 1.2>
    R2 Black:LinearColor=<rgba 0.7396 0.7396 0.7396 0.7>
    R2 BurntSienna:LinearColor=<rgba 1 0.4604 0.2292 1>
    R2 Cobalt:LinearColor=<rgba 0.2 0.35 2 2>
    R2 Crimson:LinearColor=<rgba 0.7917 0 0.0215 1>
    R2 DefaultColor:LinearColor=<rgba 1 0.6333 0 2>
    R2 ForestGreen:LinearColor=<rgba 0.0833 1 0 1>
    R2 Grey:LinearColor=<rgba 0.5469 0.5469 0.5469 1>
    R2 Lime:LinearColor=<rgba 1 2 0 1>
    R2 Orange:LinearColor=<rgba 0.901 0.3815 0.1032 1>
    R2 Pink:LinearColor=<rgba 0.6823 0.0817 0.4639 1>
    R2 Purple:LinearColor=<rgba 0.6945 0 1.3542 1>
    R2 SkyBlue:LinearColor=<rgba 0 0.551 0.8281 1>
    R2 TitaniumWhite:LinearColor=<rgba 1.4896 1.4896 1.4896 1>
    Saffron=<rgba 0.7862 0.85 0.085 2.5>
    SkyBlue=<rgba 0 0.75 1.5 2>
    TitaniumWhite=<rgba 1 1 1 2.5>
    Use R1 Painted Overrides=true
    Use R2 Painted Overrides=true
   )
   &VC_InitializeRibbon_v1(
    Color=$Particles.Color
    Custom Offset=true
    Enable Skeletal Position=true
    Lifetime=&MakeCustomFloatFromBool(Bool=$User.bIsFrontendPreview, True Float=1.25, False Float=1)
    Local Vector Offset=&AddVectorToPosition(
     Position=&MakeCustomPositionFromBool(Bool:NiagaraBool=$User.bIsFrontendPreview, True Position:NiagaraPosition=<10 0 0>, False Position:NiagaraPosition=<60 0 0>)
     Vector=$User.DTCarOffset
    )
    Offset Amount=1
    Ribbon Width:NiagaraFloat=&MakeCustomFloatFromBool(Bool=$User.bIsFrontendPreview, True Float=100, False Float=130)
   )
   &AddVelocity(Velocity=&MakeCustomVectorFromBool(Bool=$User.bIsFrontendPreview, True Vector=<-1000 0 1>, False Vector=<-1000 0 0>)) {module=AddVelocity001}
   &Color(Color=&MakeCustomLinearColorFromBool(Bool=$User.bIsPainted, True Color=$Particles.Color, False Color=<rgba 1.25 0.0224 0.0125 8>)) {enabled=false, module=Color001}
   /SetVariables(Particles.BlackLine:LinearColor=<rgba 0 0 0 2.5>) {enabled=false, isSetParameters=true}
   /SetVariables(Particles.RIBBONWIDTH2=50, Particles.RIBBONWIDTH3=50) {isSetParameters=true}
  ]
  particleUpdateScript [
   &ParticleState()
   &InheritSourceMovement(
    Applied Position Scale=&Multiply_VectorByFloat(
     Vector=<1 1 1>
     Float=&FloatFromCurve(FloatCurve=DataInterface() {propertyValues {Curve {keys [
          {
           arriveTangentWeight=0.4252
           interpMode=RCIM_Cubic
           leaveTangentWeight=0.4252
           tangentMode=RCTM_User
           tangentWeightMode=RCTWM_WeightedBoth
           time=0.15
           value=1
          }
          {interpMode=RCIM_Cubic, tangentMode=RCTM_User, time=1, value=0}
         ]}}})
    )
    ApplyForce=false
    ApplyPosition=true
    ApplyPositionFalloff=false
   )
   &SolveForcesAndVelocity(Clamp Velocity=false, Limit Acceleration=false)
   &VC_UpdateRibbon_v1(Enable Skeletal Position=true)
   &ScaleRibbonWidth(Ribbon Width Scale=&FloatFromCurve(FloatCurve=DataInterface() {propertyValues {Curve {keys [
         {arriveTangent=1.6117, interpMode=RCIM_Cubic, leaveTangent=1.6117, tangentMode=RCTM_User, time=0, value=0.6}
         {arriveTangent=-0.0567, interpMode=RCIM_Cubic, leaveTangent=-0.0567, tangentMode=RCTM_User, time=0.6, value=1.35}
         {arriveTangent=-5.6212, interpMode=RCIM_Cubic, leaveTangent=-5.6212, tangentMode=RCTM_User, time=1, value=-1}
        ]}}}))
   &DynamicMaterialParameters(
    Index 0 Param 1=1
    Index 0 Param 2=1.25
    Index 0 Param 3=&SelectFloatFromArray(Direct Array Index=$Particles.SpawnGroup, Float Selection Array=DataInterface() {propertyValues {FloatData [0,0.5]}})
    Module.Index0Param0Write=true
    Module.Index0Param1Write=true
    Module.Index0Param2Write=true
    Module.Index0Param3Write=true
   )
   &ScaleColor(
    Scale Alpha=&MakeCustomFloatFromBool(
     Bool=$User.bIsPainted
     True Float=5
     False Float=&FloatFromCurve(
      FloatCurve=DataInterface() {propertyValues {Curve {keys [
          {tangentMode=RCTM_User, time=0, value=1}
          {arriveTangent=-0.0405, interpMode=RCIM_Cubic, leaveTangent=-0.0405, tangentMode=RCTM_User, time=0.7009, value=1}
          {arriveTangent=-8.0089, interpMode=RCIM_Cubic, leaveTangent=-8.0089, tangentMode=RCTM_User, time=1, value=0}
         ]}}}
      Scale Curve=5
     )
    )
    Scale RGB=&Multiply_VectorByFloat(
     Vector=&Multiply_VectorByFloat(Vector=<1 1 1>, Float=$User.PaintedScalar)
     Float=&MakeCustomFloatFromBool(
      Bool=$User.bIsPainted
      True Float=&FloatFromCurve(FloatCurve=DataInterface() {propertyValues {Curve {keys [
           {arriveTangent=8.9191, interpMode=RCIM_Cubic, leaveTangent=8.9191, tangentMode=RCTM_User, time=0, value=5}
           {arriveTangent=-7.6025, interpMode=RCIM_Cubic, leaveTangent=-7.6025, tangentMode=RCTM_User, time=0.6, value=1}
          ]}}})
      False Float=&FloatFromCurve(FloatCurve=DataInterface() {propertyValues {Curve {keys [<at 0 5>,<at 0.5 0.75>]}}})
     )
    )
    ScaleA=true
    ScaleRGB=true
   ) {enabled=false}
   &ScaleColor(
    Scale Alpha=&MakeCustomFloatFromBool(Bool=$User.bIsPainted, True Float=30, False Float=20)
    Scale RGB=&Multiply_VectorByFloat(
     Vector=<1 1 1>
     Float=&MakeCustomFloatFromBool(
      Bool=$User.bIsPainted
      True Float=&FloatFromCurve(FloatCurve=DataInterface() {propertyValues {Curve {keys [
           {arriveTangent=-8.0973, interpMode=RCIM_Cubic, leaveTangent=-8.0973, tangentMode=RCTM_User, time=0, value=2.9}
           {arriveTangent=-2.5, interpMode=RCIM_Cubic, leaveTangent=-1.9725, tangentMode=RCTM_Break, time=0.5, value=0.5}
           {arriveTangentWeight=0.3478, interpMode=RCIM_Cubic, tangentMode=RCTM_User, time=1, value=0.4}
          ]}}})
      False Float=&FloatFromCurve(FloatCurve=DataInterface() {propertyValues {Curve {keys [
           {arriveTangent=-8.0973, interpMode=RCIM_Cubic, leaveTangent=-8.0973, tangentMode=RCTM_User, time=0, value=4}
           {arriveTangentWeight=0.3478, interpMode=RCIM_Cubic, tangentMode=RCTM_User, time=0.8, value=1}
          ]}}})
     )
    )
    ScaleA=true
    ScaleRGB=true
   ) {module=ScaleColor001}
   /SetVariables(
    Particles.R1 Color:LinearColor=&MakeCustomLinearColorFromBool(
     Bool=$User.bIsPainted
     True Color=&VC_DynamicScaleColor(
      Color to Scale:LinearColor=$Particles.Initial.R1 Color
      RGB Scale:Vector3f=&VectorFromFloat(Value=&FloatFromCurve(
        FloatCurve=DataInterface() {propertyValues {Curve {keys [{arriveTangent=-2.6577, interpMode=RCIM_Cubic, leaveTangent=-2.6577, tangentMode=RCTM_User, time=0, value=1}, <at 0.9 0 RCIM_Cubic>]}}}
        Scale Curve=2
       ))
      Alpha Scale:NiagaraFloat=80
     )
     False Color=&ColorFromCurve(ColorCurve=DataInterface() {propertyValues {
        AlphaCurve {keys [<at 0 100 RCIM_Cubic>,<at 1 100 RCIM_Cubic>]}
        BlueCurve {keys [<at 0.0846 0.5 RCIM_Cubic>,<at 0.2068 0>,<at 0.3289 0 RCIM_Cubic>]}
        GreenCurve {keys [<at 0.0846 0.5833 RCIM_Cubic>,<at 0.2068 0.4167>,<at 0.3289 0.1667 RCIM_Cubic>]}
        RedCurve {keys [<at 0.0846 1 RCIM_Cubic>,<at 0.2068 1>,<at 0.3289 1 RCIM_Cubic>]}
       }})
    )
    Particles.R2 Color:LinearColor=&MakeCustomLinearColorFromBool(
     Bool=$User.bIsPainted
     True Color=&VC_DynamicScaleColor(
      Color to Scale:LinearColor=$Particles.Initial.R2 Color
      RGB Scale:Vector3f=<1 1 1>
      Alpha Scale:NiagaraFloat=&FloatFromCurve(FloatCurve=DataInterface() {propertyValues {Curve {keys [<at 0 1 RCIM_Cubic>,<at 0.5 0 RCIM_Cubic>]}}}, Scale Curve=3)
     )
     False Color=&ColorFromCurve(ColorCurve=DataInterface() {propertyValues {
        AlphaCurve {keys [<at 0.0006 5>, {interpMode=RCIM_Cubic, leaveTangent=-9.4681, time=0.3891, value=1}, <at 0.5334 0 RCIM_Cubic>]}
        BlueCurve {keys [<at 0 0 RCIM_Cubic>,<at 0.5255 0 RCIM_Cubic>]}
        GreenCurve {keys [<at 0 0.8333 RCIM_Cubic>,<at 0.5255 0.8333 RCIM_Cubic>]}
        RedCurve {keys [<at 0 1 RCIM_Cubic>,<at 0.5255 1 RCIM_Cubic>]}
       }})
    )
    Particles.RIBBONWIDTH2:NiagaraFloat=&FloatFromCurve(
     FloatCurve=DataInterface() {propertyValues {Curve {keys [<at 0 0 RCIM_Cubic>,<at 0.3 1.25 RCIM_Cubic>,<at 0.55 0 RCIM_Cubic>]}}}
     Scale Curve=&SelectFloatFromArray(Direct Array Index=$Particles.SpawnGroup, Float Selection Array=DataInterface() {propertyValues {FloatData [65,-65]}})
    )
    Particles.RIBBONWIDTH3:NiagaraFloat=&FloatFromCurve(
     FloatCurve=DataInterface() {propertyValues {Curve {keys [<at 0.5 1 RCIM_Cubic>,<at 1 1 RCIM_Cubic>]}}}
     Scale Curve=&SelectFloatFromArray(Direct Array Index=$Particles.SpawnGroup, Float Selection Array=DataInterface() {propertyValues {FloatData [70,-70]}})
    )
   ) {isSetParameters=true}
  ]
  rendererClasses [&NiagaraRibbonRendererProperties]
  renderers [
   {rendererClass=&NiagaraRibbonRendererProperties}
   {rendererClass=&NiagaraRibbonRendererProperties, rendererIndex=1}
   {rendererClass=&NiagaraRibbonRendererProperties, rendererIndex=2}
  ]
 }
}
summary {
 systemName=NS_Boost_StylizedFlame
 emitters [
  {emitterName=NE_VC_Boost_BaseFlames, rendererClasses [&NiagaraSpriteRendererProperties]}
  {emitterName=NE_VC_SolidFlames_01, rendererClasses [&NiagaraRibbonRendererProperties]}
 ]
 userVariables [
  User.bIsFrontendPreview {type=NiagaraBool}
  User.bIsPainted {type=NiagaraBool}
  User.ColorVariantID {type=NiagaraInt32}
  User.DTCarOffset {type=Vector3f, description="", defaultValue {struct=Vector3f, value=<0 0 0>}}
  User.PaintedColor {type=LinearColor, description="", defaultValue {struct=LinearColor, value=<rgba 1 1 1 1>}}
  User.PaintedScalar {type=NiagaraFloat, defaultValue=1, description=""}
  User.PortCount {type=NiagaraInt32, defaultValue=2}
 ]
}
